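# Physics asset for the VealCadet Flare pickaxe, a two-handed melee cosmetic, that sets up a physically simulated dangling chain.
# Capsule bodies cover the handle, a small MSTR_chain root and six dyn_chain links, with a larger, angled capsule on the last link.
# Constraints join the links in series from MSTR_chain, allowing limited swing and little or no twist.
# The handle capsule has no constraint to the chain.
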PhysicsAsset /BRCosmetics/Weapons/FORT_Melee/Pickaxe_VealCadet_Flare/Meshes/VealCadet_Flare_Axe_Physics
bodies {
 MSTR_chain [{shapeType=Capsule, length=0.505, radius=0.505, shapeName=MSTR_chain_capsule, rotation=<rot 0 -90 0>}]
 Melee_TwoHanded_Handle [{shapeType=Capsule, length=58.3499, radius=10.8953, center=<0 0 29.0522>}]
 dyn_chain_01 [{shapeType=Capsule, length=1.8537, radius=1.5734, shapeName=dyn_chain_01_capsule, center=<1.4 0 0>, rotation=<rot 90 0 0>}]
 dyn_chain_02 [{shapeType=Capsule, length=3.0751, radius=2.23, shapeName=dyn_chain_02_capsule, center=<0.4 0 0>, rotation=<rot 90 0 0>}]
 dyn_chain_03 [{shapeType=Capsule, length=2.9226, radius=1.8338, shapeName=dyn_chain_03_capsule, center=<1 0 0>, rotation=<rot 90 0 0>}]
 dyn_chain_04 [{shapeType=Capsule, length=2.89, radius=1.8338, shapeName=dyn_chain_04_capsule, center=<1 0 0>, rotation=<rot 90 0 0>}]
 dyn_chain_05 [{shapeType=Capsule, length=2.8575, radius=1.8338, shapeName=dyn_chain_05_capsule, center=<1 0 0>, rotation=<rot 90 0 0>}]
 dyn_chain_06 [{shapeType=Capsule, length=6.5803, radius=2.9372, shapeName=dyn_chain_06_capsule, center=<4.5 0 0>, rotation=<rot 90 -22.0895 0>}]
}
constraints [
 {bone1Name=dyn_chain_01, bone2Name=MSTR_chain, swing1LimitDegrees=15, swing2LimitDegrees=100, twistLimitDegrees=0}
 {bone1Name=dyn_chain_02, bone2Name=dyn_chain_01, swing1LimitDegrees=35, swing2LimitDegrees=40, twistLimitDegrees=5}
 {bone1Name=dyn_chain_03, bone2Name=dyn_chain_02, swing1LimitDegrees=35, swing2LimitDegrees=45, twistLimitDegrees=5}
 {bone1Name=dyn_chain_04, bone2Name=dyn_chain_03, swing1LimitDegrees=35, swing2LimitDegrees=45, twistLimitDegrees=5}
 {bone1Name=dyn_chain_05, bone2Name=dyn_chain_04, swing1LimitDegrees=35, swing2LimitDegrees=45, twistLimitDegrees=5}
 {bone1Name=dyn_chain_06, bone2Name=dyn_chain_05, swing1LimitDegrees=35, swing2LimitDegrees=40, twistLimitDegrees=5}
]
